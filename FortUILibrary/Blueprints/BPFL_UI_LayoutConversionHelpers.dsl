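; Blueprint function library of UI layout helpers that choose between two layout values based on a boolean.
; Three functions cover float, Margin and 2D vector size values, each returning its True input or its False input according to InBool.
; The float variant uses SelectFloat, while the margin and size variants use Select nodes.
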(fn CF_SelectDesiredUIFloat
  (params (InBool :type Boolean) (TrueFloat :type Float%20%28double-precision%29) (FalseFloat :type Float%20%28double-precision%29))
  (return (Math|Float|SelectFloat TrueFloat FalseFloat (Variables|GetInBool))))

(fn CF_SelectDesiredUIMargin
  (params (InBool :type Boolean) (TrueMargin :type Margin%20Structure) (FalseMargin :type Margin%20Structure))
  (return (select (Variables|GetInBool) (Variables|GetTrueMargin) (Variables|GetFalseMargin))))

(fn CF_SelectDesiredUISize
  (params (InBool :type Boolean) (TrueSize :type Vector%202D%20Structure) (FalseSize :type Vector%202D%20Structure))
  (return (select (Variables|GetInBool) (Variables|GetTrueSize) (Variables|GetFalseSize))))
