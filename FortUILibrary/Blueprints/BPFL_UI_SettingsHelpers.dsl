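; UI helper function library that opens communication settings for chat options and reports which game UI is active.
; Save the World is detected by comparing the global UI context's SubGame to Campaign.
; Text chat, voice chat and voice method helpers funnel into one settings request, choosing tab and setting names by that check.
; The BR UI check combines a new-HUD console variable, non-Save the World, and Creative HUD enablement for spatial experiences.
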(fn IsBRGameUIEnabled
  (bind _returnvalue (Development|GetConsoleVariableBoolValue "BattleRoyalGameUI.NewHUD.Enabled"))
  (return (Math|Boolean|ANDBoolean _returnvalue (not (CallFunction|IsSavetheWorldUI (Variables|Get__WorldContext))) (or (not (Fortnite|IsSpatialGameplayExperience)) (Fortnite|IsNewHudEnabledforCreative)))))

(fn IsSaveTheWorldUI
  (bind _returnvalue (LocalPlayerSubsystems|GetFortGlobalUIContext 0))
  (Utilities|IsValid _returnvalue
    (:Is%20Valid
      (return (Utilities|Enum|Equal(Enum) (GlobalUIContext|SubGame|GetSubGame _returnvalue) "Campaign")))
    (:Is%20Not%20Valid
      (return false))))

(fn OpenSettingsCommunicationTab
  (params (OwningWidget :type Common%20User%20Widget%20Object%20Reference) (SettingToSelect :type String))
  (FortSettingsContext|RequestOpenSettings (BlueprintContext|GetFortSettingsContext) (Variables|GetOwningWidget) (select (CallFunction|IsSavetheWorldUI 0) "Communication" "CommunicationCollection") (Variables|GetSettingToSelect)))

(fn OpenSettingsToTextChat
  (params (OwningWidget :type Common%20User%20Widget%20Object%20Reference))
  (Default|OpenSettingsCommunicationTab OwningWidget (Utilities|String|SelectString "ShowTextChat" "TextChatCollection" (CallFunction|IsSavetheWorldUI 0))))

(fn OpenSettingsToVoiceChat
  (params (OwningWidget :type Common%20User%20Widget%20Object%20Reference))
  (Default|OpenSettingsCommunicationTab OwningWidget (Utilities|String|SelectString "VoiceChat_User" "VoiceChatCollection" (CallFunction|IsSavetheWorldUI 0))))

(fn OpenSettingsToVoiceChatMethod
  (params (OwningWidget :type Common%20User%20Widget%20Object%20Reference))
  (Default|OpenSettingsCommunicationTab OwningWidget (Utilities|String|SelectString "VoiceChatMethodWithMute" "VoiceChatMethod" (CallFunction|IsSavetheWorldUI 0))))
