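; UI Blueprint function library that converts dates and quantities into display text.
; Date helpers return localized "Moments ago", abbreviated "{Timespan} ago", Today, Yesterday, day-of-week or medium-date labels, depending on age and function.
; Day-based labels use GetNormalizedTimespanDifference, which truncates both the date and UTC now to midnight, so Today and Yesterday follow UTC calendar days rather than 24-hour windows.
; The quantity helper switches to shorthand text above 999.
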(fn ConvertDateTimeToTimeAgoText
  (params (DateTime :type Date%20Time%20Structure))
  (return (Utilities|String|SelectText "NSLOCTEXT(\"[BB2451C0C025C5B3E2A78AF3E6C30482]\", \"0EDB6F584EC549CE7D5AE083E57EABEC\", \"Moments ago\")" (Utilities|Text|FormatText "NSLOCTEXT(\"[BB2451C0C025C5B3E2A78AF3E6C30482]\", \"23419C784BC60A5633027AB66E4D8F9C\", \"{Timespan} ago\")" (TimeOfDay|GetTimespanasText (Math|DateTime|DateTime-DateTime (Math|DateTime|UTCNow) DateTime) false false (select (> (Math|Timespan|GetTotalDays (Math|DateTime|DateTime-DateTime (Math|DateTime|UTCNow) DateTime)) 365) "ApproximateWithYears" "ApproximateWithWeeks") "Abbreviated")) (< (Math|Timespan|GetTotalMinutes (Math|DateTime|DateTime-DateTime (Math|DateTime|UTCNow) DateTime)) 5))))

(fn ConvertDateTimeToTodayYesterdayDayOfWeekOrDate
  (params (DateTime :type Date%20Time%20Structure))
  (bind _normalizedtimespandifference (TimeHelpers|GetNormalizedTimespanDifference DateTime))
  (bind _returnvalue (Math|Timespan|GetDays _normalizedtimespandifference))
  (switch int _returnvalue
    (:Default
      (if (<= _returnvalue 7)
        (return (DisplayHelper|AsDayOftheWeek (Variables|GetDateTime)))
        (else
          (return (DisplayHelper|AsMediumDate (Variables|GetDateTime))))))
    (:0
      (return "NSLOCTEXT(\"[BB2451C0C025C5B3E2A78AF3E6C30482]\", \"7C8930F34219BEDB073FBDA190491AA9\", \"Today\")"))
    (:1
      (return "NSLOCTEXT(\"[BB2451C0C025C5B3E2A78AF3E6C30482]\", \"237F824A4D914B39A2CD9A94B5D0CCFB\", \"Yesterday\")"))))

(fn ConvertDateTimeToTodayYesterdayOrTimeAgoText
  (params (DateTime :type Date%20Time%20Structure))
  (switch int (Math|Timespan|GetDays (TimeHelpers|GetNormalizedTimespanDifference DateTime))
    (:Default
      (return (TimeHelpers|ConvertDateTimetoTimeAgoText (Variables|GetDateTime))))
    (:0
      (return "NSLOCTEXT(\"[BB2451C0C025C5B3E2A78AF3E6C30482]\", \"7C8930F34219BEDB073FBDA190491AA9\", \"Today\")"))
    (:1
      (return "NSLOCTEXT(\"[BB2451C0C025C5B3E2A78AF3E6C30482]\", \"237F824A4D914B39A2CD9A94B5D0CCFB\", \"Yesterday\")"))))

(fn ConvertQuantityToShorthandText
  (params (Quantity :type Integer))
  (return (Utilities|String|SelectText (Inventory|GetShorthandQuantity Quantity 0 1) (Utilities|Text|ToText(Integer) Quantity) (> Quantity 999))))

(fn GetNormalizedTimespanDifference
  (params (DateTime :type Date%20Time%20Structure))
  (bind (_year _month _day) (Math|DateTime|BreakDateTime (Math|DateTime|UTCNow)))
  (bind (_year_2 _month_2 _day_2) (Math|DateTime|BreakDateTime DateTime))
  ; to properly calculate Today, Yesterday, etc, we need to normalize the date to midnight.  Otherwise, if you sent a message before midnight yesterday, but it was less than 24 hours ago to now, then it would show as "today" incorrectly, instead of "yesterday."
  (LocalVariables|SetDateReceivedMidnight (Math|DateTime|MakeDateTime _year_2 _month_2 _day_2))
  (LocalVariables|SetTodayMidnight (Math|DateTime|MakeDateTime _year _month _day))
  (return (Math|DateTime|DateTime-DateTime (LocalVariables|GetTodayMidnight) (LocalVariables|GetDateReceivedMidnight))))
