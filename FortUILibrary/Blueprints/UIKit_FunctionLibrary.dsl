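; Blueprint function library of shared helpers for Fortnite's UIKit widgets.
; Button-state functions snap Trans_* scalar parameters on dynamic materials for selected, deselected, enabled and disabled looks, and another copies parameters between MIDs.
; PushSimpleUIKitModalDialog builds a viewmodel from a class, creates WBP_UIKit_Dialog, binds its button event and pushes it as a modal.
; Other helpers cover grid slot layout, widget removal, animation stopping, analytics and purchase trace IDs, input-type checks, hex-to-color parsing and value remapping.
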(fn CopyMIDParameterOverrides
  (params (Target%20MID :type Material%20Instance%20Dynamic%20Object%20Reference) (Source%20MID :type Material%20Instance%20Dynamic%20Object%20Reference))
  (if (Math|Boolean|ANDBoolean (Utilities|IsValid Target%20MID :!class /Script/Engine.KismetSystemLibrary) (Utilities|IsValid Source%20MID :!class /Script/Engine.KismetSystemLibrary) (!= Target%20MID Source%20MID))
    ;[0.06597 0.06597 0.06597 1 12] Passing internal MID parameter values to new MID to keep current visual state.
    (Rendering|Material|CopyParameterOverrides Target%20MID Source%20MID)))

(fn FireEventElementInteractedByName
  (params (Element%20Name :type String) (WidgetCallingFrom :type Widget%20Object%20Reference))
  (if (not (Utilities|String|IsEmpty Element%20Name))
    (bind _asfort_player_controller (Utilities|Casting|CastToFortPlayerController (Widget|GetOwningPlayer WidgetCallingFrom))
      (:then
        (LocalVariables|SetFuncSuccess true)
        (FlowAnalytics|FireEventElementInteracted(Non-Localized) _asfort_player_controller (Variables|GetElement%20Name)))
      (:CastFailed
        (LocalVariables|SetFuncSuccess false)))
    (return (LocalVariables|GetFuncSuccess))
    (else
      _)))

(fn GeneratePurchaseTraceId
  (params (OfferId :type String))
  (bind _returnvalue (MCPContext|GeneratePurchaseUXTraceId (BlueprintContext|GetFortMcpContext) OfferId ""))
  (return (FlowAnalytics|AddString(Non-Localized) (Utilities|Struct|MakeFlowAnalyticsAdditionalInfo) "PurchaseUXTraceId" _returnvalue)))

(fn GetPurchaseTraceId
  (return (FlowAnalytics|AddString(Non-Localized) (Utilities|Struct|MakeFlowAnalyticsAdditionalInfo) "PurchaseUXTraceId" (MCPContext|GetPurchaseUXTraceId (BlueprintContext|GetFortMcpContext) ""))))

(fn GetSmallTextScalingMultiplier
  (bind _returnvalue (Development|GetConsoleVariableFloatValue "CommonUI.SmallTextScaling"))
  (return _returnvalue (and (> _returnvalue 0) (not (DisplayHelper|IsUsingCompactUI)))))

(fn HexToColor
  (params (HEX :type String))
  (bind _l_hextodec (LocalVariables|GetL_HexToDec))
  (bind _l_hextodec_2 (LocalVariables|GetL_HexToDec))
  ;[0 0 0] Channel parsing
  (for _index (range 4)
    ;[0 0 0] MakeColor
    (switch int _index
      (:0
        (bind _structout (Utilities|Struct|SetmembersinLinearColor (LocalVariables|GetL_Color) (/ (+ (* (Math|Integer|SelectInt 15 (Utilities|String|FindSubstring _l_hextodec_2 (Utilities|String|GetSubstring HEX (* _index 2))) (Utilities|String|IsEmpty (Utilities|String|GetSubstring HEX (* _index 2)))) 16) (Math|Integer|SelectInt 15 (Utilities|String|FindSubstring _l_hextodec (Utilities|String|GetSubstring HEX (+ (* _index 2) 1))) (Utilities|String|IsEmpty (Utilities|String|GetSubstring HEX (+ (* _index 2) 1))))) 255.0)))
        (LocalVariables|SetL_Color _structout))
      (:1
        (bind _structout_2 (Utilities|Struct|SetmembersinLinearColor (LocalVariables|GetL_Color) (/ (+ (* (Math|Integer|SelectInt 15 (Utilities|String|FindSubstring _l_hextodec_2 (Utilities|String|GetSubstring HEX (* _index 2))) (Utilities|String|IsEmpty (Utilities|String|GetSubstring HEX (* _index 2)))) 16) (Math|Integer|SelectInt 15 (Utilities|String|FindSubstring _l_hextodec (Utilities|String|GetSubstring HEX (+ (* _index 2) 1))) (Utilities|String|IsEmpty (Utilities|String|GetSubstring HEX (+ (* _index 2) 1))))) 255.0)))
        (LocalVariables|SetL_Color _structout_2))
      (:2
        (bind _structout_3 (Utilities|Struct|SetmembersinLinearColor (LocalVariables|GetL_Color) (/ (+ (* (Math|Integer|SelectInt 15 (Utilities|String|FindSubstring _l_hextodec_2 (Utilities|String|GetSubstring HEX (* _index 2))) (Utilities|String|IsEmpty (Utilities|String|GetSubstring HEX (* _index 2)))) 16) (Math|Integer|SelectInt 15 (Utilities|String|FindSubstring _l_hextodec (Utilities|String|GetSubstring HEX (+ (* _index 2) 1))) (Utilities|String|IsEmpty (Utilities|String|GetSubstring HEX (+ (* _index 2) 1))))) 255.0)))
        (LocalVariables|SetL_Color _structout_3))
      (:3
        (bind _structout_4 (Utilities|Struct|SetmembersinLinearColor (LocalVariables|GetL_Color) (/ (+ (* (Math|Integer|SelectInt 15 (Utilities|String|FindSubstring _l_hextodec_2 (Utilities|String|GetSubstring HEX (* _index 2))) (Utilities|String|IsEmpty (Utilities|String|GetSubstring HEX (* _index 2)))) 16) (Math|Integer|SelectInt 15 (Utilities|String|FindSubstring _l_hextodec (Utilities|String|GetSubstring HEX (+ (* _index 2) 1))) (Utilities|String|IsEmpty (Utilities|String|GetSubstring HEX (+ (* _index 2) 1))))) 255.0)))
        (LocalVariables|SetL_Color _structout_4))))
  ;[0 0 0] Result
  (return (LocalVariables|GetL_Color)))

(fn InstantTransitionToDisabled
  (params (MID :type Material%20Instance%20Dynamic%20Object%20Reference) (IsSelected :type Boolean))
  (if IsSelected
    (Rendering|Material|SetScalarParameterValue MID "Trans_SelectedDisabled" 1.0 :!class /Script/Engine.MaterialInstanceDynamic)
    (else
      (Rendering|Material|SetScalarParameterValue MID "Trans_Disabled" 1.0 :!class /Script/Engine.MaterialInstanceDynamic))))

(fn InstantTransitionToEnabled
  (params (MID :type Material%20Instance%20Dynamic%20Object%20Reference) (IsSelected :type Boolean))
  (if (not IsSelected)
    (Rendering|Material|SetScalarParameterValue MID "Trans_Hovered" 0.0 :!class /Script/Engine.MaterialInstanceDynamic)
    (Rendering|Material|SetScalarParameterValue MID "Trans_Focused" 0.0 :!class /Script/Engine.MaterialInstanceDynamic)
    (Rendering|Material|SetScalarParameterValue MID "Trans_Pressed" 0.0 :!class /Script/Engine.MaterialInstanceDynamic)
    (Rendering|Material|SetScalarParameterValue MID "Trans_Disabled" 0.0 :!class /Script/Engine.MaterialInstanceDynamic)
    (Rendering|Material|SetScalarParameterValue MID "Trans_SelectedIdle" 0.0 :!class /Script/Engine.MaterialInstanceDynamic))
  (Rendering|Material|SetScalarParameterValue MID "Trans_SelectedHovered" 0.0 :!class /Script/Engine.MaterialInstanceDynamic)
  (Rendering|Material|SetScalarParameterValue MID "Trans_SelectedFocused" 0.0 :!class /Script/Engine.MaterialInstanceDynamic)
  (Rendering|Material|SetScalarParameterValue MID "Trans_SelectedPressed" 0.0 :!class /Script/Engine.MaterialInstanceDynamic)
  (Rendering|Material|SetScalarParameterValue MID "Trans_SelectedDisabled" 0.0 :!class /Script/Engine.MaterialInstanceDynamic))

(fn InstantTransitionToSelected
  (params (MID :type Material%20Instance%20Dynamic%20Object%20Reference))
  (Rendering|Material|SetScalarParameterValue MID "Trans_SelectedIdle" 1.0 :!class /Script/Engine.MaterialInstanceDynamic)
  (Rendering|Material|SetScalarParameterValue MID "Trans_SelectedHovered" 0.0 :!class /Script/Engine.MaterialInstanceDynamic)
  (Rendering|Material|SetScalarParameterValue MID "Trans_SelectedFocused" 0.0 :!class /Script/Engine.MaterialInstanceDynamic)
  (Rendering|Material|SetScalarParameterValue MID "Trans_SelectedPressed" 0.0 :!class /Script/Engine.MaterialInstanceDynamic)
  (Rendering|Material|SetScalarParameterValue MID "Trans_SelectedDisabled" 0.0 :!class /Script/Engine.MaterialInstanceDynamic))

(fn IsInputType
  (params (InputType :type ECommonInputType%20Enum :default "MouseAndKeyboard") (IsNot :type Boolean))
  (return (select (Variables|GetIsNot) (not (Utilities|Enum|Equal(Enum) InputType (CommonInputSubsystem|GetCurrentInputType (LocalPlayerSubsystems|GetCommonInputSubsystem 0)))) (Utilities|Enum|Equal(Enum) InputType (CommonInputSubsystem|GetCurrentInputType (LocalPlayerSubsystems|GetCommonInputSubsystem 0))))))

(fn PushSimpleUIKitModalDialog
  (params (Context :type User%20Widget%20Object%20Reference) (VM%20Class :type UI%20Kit%20Dialog%20Viewmodel%20Class%20Reference) (OnDialogButtonClicked :type Delegate%20%28by%20ref%29))
  (bind _context (Variables|GetContext))
  (Utilities|IsValid Context
    (:Is%20Valid
      (if (Utilities|IsValidClass VM%20Class)
        (bind _returnvalue (Game|ConstructObjectfromClass VM%20Class _context))
        (bind _returnvalue_2 (UserInterface|CreateWidget 0 "/FortUILibrary/UIKit/Dialogs/WBP_UIKit_Dialog.WBP_UIKit_Dialog_C" (Widget|GetOwningPlayer _context) _returnvalue))
        (Dispatchers|BindEventtoDialogButtonClickedDispatcher (Variables|/FortUILibrary/UIKit/Dialogs/WBP_UIKit_Dialog.WBP_UIKit_Dialog_C|GetDialog_Internal _returnvalue_2) (Variables|GetOnDialogButtonClicked))
        (FortUIManagement|TryPushWidgetasModalPanel (Variables|GetContext) _returnvalue_2)
        (return _returnvalue_2 _returnvalue)))))

(fn RemapValueToInterval
  (params (Input :type Float%20%28double-precision%29) (InputLow :type Float%20%28double-precision%29) (InputHigh :type Float%20%28double-precision%29))
  (return (Math|Float|Clamp(Float) (/ (- Input InputLow) (- InputHigh InputLow)))))

(fn RemoveWidget
  (params (Widget :type Widget%20Object%20Reference))
  (bind _returnvalue (Widget|GetParent Widget))
  (Utilities|IsValid Widget
    (:Is%20Valid
      (Utilities|IsValid _returnvalue
        (:Is%20Valid
          (Widget|Panel|RemoveChild _returnvalue Widget))))))

(fn SetGridSlotAlignment
  (params (Widget :type Widget%20Object%20Reference) (Horizontal :type EHorizontalAlignment%20Enum :default "HAlign_Fill") (Vertical :type EVerticalAlignment%20Enum :default "VAlign_Fill"))
  (Utilities|IsValid Widget
    (:Is%20Valid
      (bind _asgrid_slot (Utilities|Casting|CastToGridSlot (Variables|/Script/UMG.Widget|GetSlot Widget)))
      (Layout|GridSlot|SetHorizontalAlignment _asgrid_slot Horizontal)
      (Layout|GridSlot|SetVerticalAlignment _asgrid_slot Vertical))))

(fn SetGridSlotColumnAndRow
  (params (Widget :type Widget%20Object%20Reference) (Row :type Integer) (Row%20Span :type Integer) (Column :type Integer) (Column%20Span :type Integer))
  (Utilities|IsValid Widget
    (:Is%20Valid
      (bind _asgrid_slot (Utilities|Casting|CastToGridSlot (Variables|/Script/UMG.Widget|GetSlot Widget)))
      (Layout|GridSlot|SetRow _asgrid_slot Row)
      (Layout|GridSlot|SetRowSpan _asgrid_slot Row%20Span)
      (Layout|GridSlot|SetColumn _asgrid_slot Column)
      (Layout|GridSlot|SetColumnSpan _asgrid_slot Column%20Span))))

(fn SetGridSlotLayer
  (params (Widget :type Widget%20Object%20Reference) (Layer :type Integer))
  (bind _asgrid_slot (Utilities|Casting|CastToGridSlot (Variables|/Script/UMG.Widget|GetSlot Widget)))
  (Layout|GridSlot|SetLayer _asgrid_slot Layer))

(fn SetGridSlotNudge
  (params (Widget :type Widget%20Object%20Reference) (Nudge :type Vector%202D%20Structure))
  (Utilities|IsValid Widget
    (:Is%20Valid
      (bind _asgrid_slot (Utilities|Casting|CastToGridSlot (Variables|/Script/UMG.Widget|GetSlot Widget)))
      (Layout|GridSlot|SetNudge _asgrid_slot Nudge))))

(fn SetGridSlotPadding
  (params (Widget :type Widget%20Object%20Reference) (Padding :type Margin%20Structure))
  (Utilities|IsValid Widget
    (:Is%20Valid
      (bind _asgrid_slot (Utilities|Casting|CastToGridSlot (Variables|/Script/UMG.Widget|GetSlot Widget)))
      (Layout|BorderSlot|SetPadding _asgrid_slot Padding :!class /Script/UMG.GridSlot))))

(fn StopUnfinishedAnimation
  (params (Target :type User%20Widget%20Object%20Reference) (Animation :type Widget%20Animation%20Object%20Reference))
  (if (UserInterface|Animation|IsAnimationPlaying (Variables|GetTarget) (Variables|GetAnimation))
    (UserInterface|Animation|StopAnimation (Variables|GetTarget) (Variables|GetAnimation)))
  (return))

(fn TransitionToDeselectedFocused
  (params (MID :type Material%20Instance%20Dynamic%20Object%20Reference))
  (Rendering|Material|SetScalarParameterValue MID "Trans_SelectedIdle" 0.0 :!class /Script/Engine.MaterialInstanceDynamic)
  (Rendering|Material|SetScalarParameterValue MID "Trans_SelectedHovered" 0.0 :!class /Script/Engine.MaterialInstanceDynamic)
  (Rendering|Material|SetScalarParameterValue MID "Trans_SelectedPressed" 0.0 :!class /Script/Engine.MaterialInstanceDynamic)
  (Rendering|Material|SetScalarParameterValue MID "Trans_SelectedDisabled" 0.0 :!class /Script/Engine.MaterialInstanceDynamic)
  (Rendering|Material|SetScalarParameterValue MID "Trans_Pressed" 0.0 :!class /Script/Engine.MaterialInstanceDynamic)
  (Rendering|Material|SetScalarParameterValue MID "Trans_Disabled" 0.0 :!class /Script/Engine.MaterialInstanceDynamic))

(fn TransitionToDeselectedIdle
  (params (MID :type Material%20Instance%20Dynamic%20Object%20Reference))
  (Rendering|Material|SetScalarParameterValue MID "Trans_SelectedHovered" 0.0 :!class /Script/Engine.MaterialInstanceDynamic)
  (Rendering|Material|SetScalarParameterValue MID "Trans_SelectedFocused" 0.0 :!class /Script/Engine.MaterialInstanceDynamic)
  (Rendering|Material|SetScalarParameterValue MID "Trans_SelectedPressed" 0.0 :!class /Script/Engine.MaterialInstanceDynamic)
  (Rendering|Material|SetScalarParameterValue MID "Trans_SelectedDisabled" 0.0 :!class /Script/Engine.MaterialInstanceDynamic)
  (Rendering|Material|SetScalarParameterValue MID "Trans_Hovered" 0.0 :!class /Script/Engine.MaterialInstanceDynamic)
  (Rendering|Material|SetScalarParameterValue MID "Trans_Focused" 0.0 :!class /Script/Engine.MaterialInstanceDynamic)
  (Rendering|Material|SetScalarParameterValue MID "Trans_Pressed" 0.0 :!class /Script/Engine.MaterialInstanceDynamic)
  (Rendering|Material|SetScalarParameterValue MID "Trans_Disabled" 0.0 :!class /Script/Engine.MaterialInstanceDynamic))
