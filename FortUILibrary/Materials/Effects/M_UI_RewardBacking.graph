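# UI material that draws an animated purple glow, named as a reward backing, using alpha-composite blending with vertex color tint.
# A pulsing radial gradient and mirrored-UV edge falloff shape the glow, broken up by two screen-space panning noise samples.
# Purple is raised to an exponent blended between white and an oscillating color through rotating UVs, varying the hue.
# The mask drives color and a weaker alpha output, keeping the glow mostly additive.
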Material /FortUILibrary/Materials/Effects/M_UI_RewardBacking
graph {
 properties {
  bAutomaticallySetUsageInEditor=false
  bUsedWithStaticMesh=false
  blendMode=BLEND_AlphaComposite
  materialDomain=MD_UI
  usedShadingModels=Default Lit
 }
 nodes [
  Constant() {r=1}
  n1 = VertexColor()
  n2 = Constant3Vector() {constant=<rgba 1 1 1 1>}
  n3 = MaterialFunctionCall!(PrimaryUV, SecondaryUV, UVMix, PixelSize) {materialFunction=/FortUILibrary/MaterialFunctions/CoreUI/MF_UI_CoreUI_OscillatingColor}
  n4 = Time()
  n5 = ComponentMask(None=Rotator(
    Coordinate=MaterialFunctionCall() {materialFunction=&GetUserInterfaceUV}.Normalized UV
    Time=Multiply(A=@n4) {constB=0.333}
   ) {speed=6.2832}) {r=true}
  n6 = Power(
   Base=Constant3Vector() {constant=<rgba 0.4501 0.1094 1 1>}
   Exp=NamedRerouteDeclaration(None=LinearInterpolate(A=@n2, B=@n3.Result, Alpha=@n5) {constA=1}) {name=RiftOscillatingColors, nodeColor=<rgba 0 0 0 1>}
  )
  n7 = LinearInterpolate(Alpha=MaterialFunctionCall(Speed=Constant() {r=0.1}) {materialFunction=&DebugTimeSine}.Result) {constA=0.9}
  n8 = SmoothStep(Value=MaterialFunctionCall(UV=MaterialFunctionCall(Texture Scale=@n7) {materialFunction=&ScaleUVsByCenter}.UVs) {materialFunction=&MF_UI_RadialGradient}.Smooth)
  n9 = SmoothStep(
   Min=ScalarParameter() {parameterName=Falloff_MinStep}
   Max=ScalarParameter() {parameterName=Falloff_MaxStep}
   Value=MaterialFunctionCall!(
    Repeat Count=Constant2Vector() {r=2}
    texcoord=MaterialFunctionCall() {materialFunction=&GetUserInterfaceUV}.Normalized UV
   ) {materialFunction=/Game/Environments/Papaya/Materials/MaterialFunctions/MF_UVMirror}.Result
  )
  n10 = MaterialFunctionCall(UVs=ScreenPosition().ViewportUV, Texture Scale=Constant() {r=1}) {materialFunction=&ScaleUVsByCenter}
  n11 = TextureObject() {samplerType=SAMPLERTYPE_Grayscale, texture=&LowResBlurredNoise}
  n12 = MaterialFunctionCall(UVs=ScreenPosition().ViewportUV, Texture Scale=Constant() {r=3}) {materialFunction=&ScaleUVsByCenter}
  n13 = Multiply(
   A=TextureSample(UVs=Panner(Coordinate=@n10.UVs) {speedX=-0.05}, Tex=@n11) {samplerType=SAMPLERTYPE_Grayscale}.R
   B=TextureSample(UVs=Panner(Coordinate=@n12.UVs) {speedX=0.025, speedY=0.1}, Tex=@n11) {samplerType=SAMPLERTYPE_Grayscale}.R
  )
  n14 = Multiply(A=ComponentMask(None=Multiply(A=@n8, B=@n9)) {r=true}, B=Saturate(None=Multiply(A=@n13) {constB=3}))
  n15 = Multiply(A=@n14) {constB=0.675}
  Multiply(A=@n1, B=Multiply(A=@n1.A, B=Multiply(A=@n6, B=@n15)))
  Multiply(A=@n1.A, B=Multiply(A=@n14) {constB=0.2})
 ]
}
