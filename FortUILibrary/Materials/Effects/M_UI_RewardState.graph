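# UI material for a reward-state widget, drawing an animated glowing rounded panel with a hold-progress effect and fadeout.
# Screen-space blurred noise builds the background, alongside an oscillating purple tint, edge light and outer glow masks.
# Two layers of randomized, rotating SDF triangles scroll over tiled grids, while HoldRatio and Effect Rotation parameters serve a sweeping edge mask.
# Vertex transform padding supplies pixel size and scale values, and the result composites with vertex color and Opacity.
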Material /FortUILibrary/Materials/Effects/M_UI_RewardState
graph {
 properties {
  bAutomaticallySetUsageInEditor=false
  bUsedWithStaticMesh=false
  blendMode=BLEND_AlphaComposite
  materialDomain=MD_UI
  usedShadingModels=Default Lit
 }
 nodes [
  Constant2Vector() {g=2.5, r=2.5}
  Constant() {r=1}
  n1 = NamedRerouteUsage()
  n2 = NamedRerouteUsage()
  n3 = Multiply(A=Time()) {constB=1.6}
  n4 = NamedRerouteUsage()
  n5 = AppendVector(A=Add(A=Frac(None=@n3), B=LinearInterpolate(Alpha=@n4) {constB=0.5}), B=Constant() {r=-0.15})
  n6 = Multiply(A=Frac(None=Add(A=@n2, B=@n5)), B=Constant() {r=2}) {constB=2}
  n7 = Frac(None=@n6)
  n8 = Add(A=@n7, B=Constant2Vector() {g=-0.12}) {constB=0}
  n9 = Add(A=NamedRerouteUsage(), B=Multiply(A=Time()) {constB=-0.05})
  n10 = NamedRerouteUsage()
  n11 = Multiply(
   A=Subtract(A=Rotator(Coordinate=@n8, Time=@n9) {speed=6.2832}) {constB=0.5}
   B=AppendVector(A=LinearInterpolate(Alpha=@n10) {constA=1.525, constB=0.525}, B=LinearInterpolate(Alpha=@n10) {constA=1.425, constB=2})
  )
  n12 = NamedRerouteUsage()
  n13 = MaterialFunctionCall(DistanceField=MaterialFunctionCall!(UVs=Add(A=@n11) {constB=0.5}, Radius=LinearInterpolate(Alpha=@n12) {constA=1.15, constB=0.5}, Sides=Constant() {r=3}, InnerAngle) {materialFunction=/FortUILibrary/MaterialFunctions/Shapes/SDFs/MF_SDF_RegularPolygon}.SDF) {materialFunction=&MF_UI_AntiAliasing}
  LinearInterpolate(A=@n1, Alpha=@n13.Inner)
  n14 = VertexColor()
  n15 = Multiply(A=NamedRerouteUsage(), B=NamedRerouteUsage())
  n16 = ScreenPosition()
  n17 = Constant() {r=0.375}
  n18 = Panner(Coordinate=Rotator(Coordinate=MaterialFunctionCall(UVs=@n16.ViewportUV, Texture Scale=@n17) {materialFunction=&ScaleUVsByCenter}.UVs) {speed=-0.125}) {speedX=-0.05}
  n19 = TextureObject() {samplerType=SAMPLERTYPE_Grayscale, texture=&LowResBlurredNoise}
  n20 = MaterialFunctionCall() {materialFunction=&DebugTimeSine}
  n21 = MaterialFunctionCall(UVs=@n16.ViewportUV, Texture Scale=Constant() {r=3}) {materialFunction=&ScaleUVsByCenter}
  n22 = Multiply(
   A=Power(
    Base=TextureSample(UVs=@n18, Tex=@n19) {samplerType=SAMPLERTYPE_Grayscale}.R
    Exp=LinearInterpolate(Alpha=@n20.Result) {constA=3.5, constB=4}
   ) {constExponent=4}
   B=TextureSample(UVs=Panner(Coordinate=@n21.UVs) {speedX=0.025, speedY=0.1}, Tex=@n19) {samplerType=SAMPLERTYPE_Grayscale}.R
  )
  n23 = Add(A=NamedRerouteUsage(), B=@n22)
  n24 = ComponentMask(None=MaterialFunctionCall() {materialFunction=&GetUserInterfaceUV}.Normalized UV) {g=true}
  n25 = LinearInterpolate(Alpha=@n24) {constA=9, constB=18}
  n26 = Saturate(None=Multiply(A=Multiply(A=@n22, B=@n25) {constB=3}, B=NamedRerouteUsage()))
  n27 = LinearInterpolate(A=@n26, B=LinearInterpolate(Alpha=@n26) {constA=0.5}, Alpha=NamedRerouteUsage())
  n28 = Power(Base=@n24) {constExponent=8}
  n29 = Multiply(
   A=NamedRerouteDeclaration(None=Saturate(None=@n23)) {name=Background Color, nodeColor=<rgba 0 0 0 1>}
   B=LinearInterpolate(A=NamedRerouteUsage(), Alpha=LinearInterpolate(A=@n27, Alpha=@n28))
  )
  n30 = Multiply(A=@n29) {constB=4}
  n31 = NamedRerouteUsage()
  n32 = NamedRerouteUsage()
  n33 = Add(
   A=LinearInterpolate(A=@n29, B=Saturate(None=@n30), Alpha=NamedRerouteUsage())
   B=Multiply(A=Multiply(A=@n31) {constB=1.5}, B=Multiply(A=@n32, B=@n32))
  )
  n34 = Saturate(None=Add(A=LinearInterpolate(Alpha=@n15) {constB=0.1}, B=Saturate(None=@n33)))
  n35 = NamedRerouteUsage()
  n36 = Multiply(A=@n14.A, B=LinearInterpolate(A=Multiply(A=@n34, B=@n35), B=NamedRerouteUsage(), Alpha=NamedRerouteUsage()))
  n37 = ScalarParameter() {defaultValue=1, parameterName=Opacity, sliderMax=1}
  n38 = NamedRerouteUsage()
  Multiply(A=Multiply(A=Multiply(A=@n14, B=@n36), B=@n37), B=@n38)
  n39 = NamedRerouteDeclaration(None=MaterialFunctionCall() {materialFunction=&GetUserInterfaceUV}.Normalized UV) {name=UV, nodeColor=<rgba 0 0 0 1>}
  n40 = Multiply(A=Time()) {constB=0.333}
  n41 = LinearInterpolate(
   A=Constant3Vector() {constant=<rgba 1 1 1 1>}
   B=MaterialFunctionCall!(PrimaryUV, SecondaryUV, UVMix, PixelSize) {materialFunction=/FortUILibrary/MaterialFunctions/CoreUI/MF_UI_CoreUI_OscillatingColor}.Result
   Alpha=ComponentMask(None=Rotator(Coordinate=@n39, Time=@n40) {speed=6.2832}) {r=true}
  ) {constA=1}
  NamedRerouteDeclaration(None=Power(
    Base=Constant3Vector() {constant=<rgba 0.4501 0.1094 1 1>}
    Exp=NamedRerouteDeclaration(None=@n41) {name=RiftOscillatingColors, nodeColor=<rgba 0 0 0 1>}
   )) {name=Coolor, nodeColor=<rgba 0 0 0 1>}
  n42 = NamedRerouteUsage()
  n43 = ScalarParameter() {defaultValue=20, parameterName=CornerRadius}
  n44 = NamedRerouteUsage()
  n45 = Max(A=ComponentMask(None=@n44) {r=true}, B=ComponentMask(None=@n44) {g=true})
  NamedRerouteDeclaration(None=MaterialFunctionCall(
    Size=Divide(B=@n42)
    Corner Radius=Multiply(A=@n43, B=@n45)
    Radius in Pixels=ScalarParameter() {defaultValue=1, parameterName=RadiusInPixels, sliderMax=1}
    Pixel Size=NamedRerouteUsage()
   ) {materialFunction=&MF_UI_CoreUI_RoundedMask}.Shape Mask) {name=Shape Mask, nodeColor=<rgba 0 0 0 1>}
  n46 = NamedRerouteUsage()
  n47 = Add(A=NamedRerouteUsage(), B=NamedRerouteUsage())
  n48 = NamedRerouteDeclaration(None=Multiply(A=NamedRerouteUsage(), B=Saturate(None=@n47))) {name=Masked Triangles, nodeColor=<rgba 0 0 0 1>}
  NamedRerouteDeclaration(None=Saturate(None=Add(A=@n46, B=@n48))) {name=Hold Effect Mask, nodeColor=<rgba 0 0 0 1>}
  NamedRerouteDeclaration(None=MaterialFunctionCall!(UV=Floor(None=@n6), Seed=ScalarParameter() {defaultValue=0.1, parameterName=Triangle 02 Seed}) {materialFunction=&MF_UI_Random}.Result) {name=Random Values 02, nodeColor=<rgba 0 0 0 1>}
  n49 = Max(A=Abs(None=DDX(Value=@n7)), B=Abs(None=DDY(Value=@n7)))
  n50 = Round(None=Max(A=ComponentMask(None=@n49) {r=true}, B=ComponentMask(None=@n49) {g=true}))
  n51 = Multiply(A=@n13.Inner, B=@n1)
  n52 = Saturate(None=Sine(None=@n1))
  NamedRerouteDeclaration(None=Multiply(A=OneMinus(None=@n50), B=Multiply(A=@n51, B=@n52)) {constA=1}) {name=Random Triangles 02, nodeColor=<rgba 0 0 0 1>}
  n53 = MaterialFunctionCall!(Anchor, Translation, Padding=Constant() {r=32}, Scale=Constant2Vector() {g=1, r=1}) {materialFunction=/FortUILibrary/MaterialFunctions/Basics/MF_UI_VertexTransform}
  NamedRerouteDeclaration(None=@n53.New Pixel Size) {name=New Pixel Size, nodeColor=<rgba 0 0 0 1>}
  NamedRerouteDeclaration(None=@n53.Total Scale Factor) {name=Total Scale Factor, nodeColor=<rgba 0 0 0 1>}
  n54 = MaterialFunctionCall(
   UVs=MaterialFunctionCall() {materialFunction=&GetUserInterfaceUV}.Normalized UV
   Texture Scale=Divide(B=NamedRerouteUsage())
  ) {materialFunction=&ScaleUVsByCenter}
  n55 = Abs(None=ConstantBiasScale(None=MaterialFunctionCall!(UVs=@n54.UVs) {materialFunction=/FortUILibrary/MaterialFunctions/Basics/MF_UI_9SliceUVs}.UVs) {bias=-0.5, scale=2})
  n56 = ComponentMask(None=@n55) {r=true}
  n57 = ComponentMask(None=@n55) {g=true}
  NamedRerouteDeclaration(None=Power(Base=LinearInterpolate(A=@n56, Alpha=@n57)) {constExponent=32}) {name=Min Light Mask, nodeColor=<rgba 0 0 0 1>}
  n58 = OneMinus(None=NamedRerouteUsage())
  n59 = MaterialFunctionCall() {materialFunction=&GetUserInterfaceUV}
  n60 = Power(Base=Abs(None=ConstantBiasScale(None=@n59.Normalized UV) {bias=-0.5, scale=2}))
  n61 = ComponentMask(None=@n60) {r=true}
  n62 = ComponentMask(None=@n60) {g=true}
  n63 = SmoothStep(Value=OneMinus(None=LinearInterpolate(A=@n61, Alpha=@n62)))
  NamedRerouteDeclaration(None=Multiply(A=Multiply(A=@n58, B=@n63)) {constB=0.175}) {name=Outer Glow Mask, nodeColor=<rgba 0 0 0 1>}
  NamedRerouteDeclaration(None=OneMinus(None=ScalarParameter() {parameterName=Fadeout_Amount})) {name=Fadeout_Amount, nodeColor=<rgba 0 0 0 1>}
  n64 = Multiply(A=@n55, B=@n55)
  n65 = ComponentMask(None=@n64) {r=true}
  n66 = ComponentMask(None=@n64) {g=true}
  NamedRerouteDeclaration(None=Power(Base=LinearInterpolate(A=@n65, Alpha=@n66)) {constExponent=8}) {name=Effect Mask, nodeColor=<rgba 0 0 0 1>}
  NamedRerouteDeclaration(None=ScalarParameter() {parameterName=HoldRatio, sliderMax=1}) {name=HoldRatio, nodeColor=<rgba 0 0 0 1>}
  n67 = ScalarParameter() {parameterName=Effect Rotation}
  n68 = NamedRerouteUsage()
  n69 = Add(
   A=Rotator(Coordinate=NamedRerouteUsage(), Time=NamedRerouteDeclaration(None=@n67) {name=Effect Rotation, nodeColor=<rgba 0 0 0 1>}) {speed=6.2832}
   B=AppendVector(A=LinearInterpolate(Alpha=@n68) {constA=1, constAlpha=0, constB=-0.1}, B=Constant())
  )
  n70 = Saturate(None=ComponentMask(None=@n69) {r=true})
  n71 = NamedRerouteDeclaration(None=OneMinus(None=Floor(None=@n70))) {name=Hold Mask, nodeColor=<rgba 0 0 0 1>}
  n72 = NamedRerouteUsage()
  n73 = ComponentMask(None=Add(A=@n69, B=Constant2Vector() {r=-1})) {r=true}
  n74 = Power(Base=Min(A=OneMinus(None=@n73), B=@n70)) {constExponent=16}
  n75 = LinearInterpolate(
   A=Multiply(A=MaterialFunctionCall("Value (0-1)"=@n72) {materialFunction=&MF_UI_EaseCurves}.EaseOutCubic, B=@n74)
   B=@n74
   Alpha=NamedRerouteUsage()
  )
  NamedRerouteDeclaration(None=LinearInterpolate(A=Multiply(A=@n71) {constB=0.1}, Alpha=NamedRerouteDeclaration(None=@n75) {name=Edge Falloff Mask, nodeColor=<rgba 0 0 0 1>})) {name=Hold Edge Mask, nodeColor=<rgba 0 0 0 1>}
  n76 = ScreenPosition()
  n77 = Multiply(A=NamedRerouteUsage()) {constB=-1}
  n78 = Multiply(A=Subtract(A=Rotator(Coordinate=@n76.ViewportUV, Time=@n77) {speed=6.2832}) {constB=0.5}) {constB=4.5}
  n79 = NamedRerouteDeclaration(None=MaterialFunctionCall!(ScreenUV=Add(A=@n78) {constB=0.5}, TextureSize, ScalePivotPoint) {materialFunction=/FortUILibrary/MaterialFunctions/Basics/MF_UI_ScreenSpaceUVScaling}.ScaleY) {name=Scale Y, nodeColor=<rgba 0 0 0 1>}
  n80 = Multiply(A=Time())
  n81 = NamedRerouteUsage()
  n82 = Multiply(A=@n81, B=@n81)
  n83 = AppendVector(A=Add(A=Frac(None=@n80), B=LinearInterpolate(Alpha=@n82) {constB=0.175}), B=Constant())
  n84 = Multiply(A=Frac(None=Add(A=@n79, B=@n83))) {constB=4}
  NamedRerouteDeclaration(None=MaterialFunctionCall(UV=Floor(None=@n84), Seed=ScalarParameter() {defaultValue=0.1, parameterName=Triangle 01 Seed}) {materialFunction=&MF_UI_Random}.Result) {name=Random Values 01, nodeColor=<rgba 0 0 0 1>}
  n85 = Frac(None=@n84)
  n86 = Max(A=Abs(None=DDX(Value=@n85)), B=Abs(None=DDY(Value=@n85)))
  n87 = Round(None=Max(A=ComponentMask(None=@n86) {r=true}, B=ComponentMask(None=@n86) {g=true}))
  n88 = Time()
  n89 = Rotator(Coordinate=@n85, Time=Add(A=NamedRerouteUsage(), B=Multiply(A=@n88) {constB=0.025})) {speed=6.2832}
  n90 = NamedRerouteUsage()
  n91 = LinearInterpolate(Alpha=@n90) {constA=0.5, constB=0.8}
  n92 = LinearInterpolate(Alpha=@n90) {constA=0.5, constB=2}
  n93 = Add(A=Multiply(A=Subtract(A=@n89) {constB=0.5}, B=AppendVector(A=@n91, B=@n92))) {constB=0.5}
  n94 = LinearInterpolate(Alpha=NamedRerouteUsage()) {constA=0.25, constB=0.8}
  n95 = Constant() {r=3}
  n96 = NamedRerouteUsage()
  n97 = Multiply(
   A=MaterialFunctionCall(DistanceField=MaterialFunctionCall(UVs=@n93, Radius=@n94, Sides=@n95) {materialFunction=&MF_SDF_RegularPolygon}.SDF) {materialFunction=&MF_UI_AntiAliasing}.Inner
   B=@n96
  )
  n98 = Saturate(None=Round(None=@n96))
  NamedRerouteDeclaration(None=Multiply(A=OneMinus(None=@n87), B=Multiply(A=@n97, B=@n98)) {constA=1}) {name=Random Triangles 01, nodeColor=<rgba 0 0 0 1>}
  n99 = NamedRerouteUsage()
  n100 = LinearInterpolate(A=NamedRerouteUsage(), Alpha=LinearInterpolate(A=LinearInterpolate(Alpha=@n99) {constB=0.375}, Alpha=NamedRerouteUsage()) {constB=0.5})
  n101 = NamedRerouteUsage()
  n102 = Multiply(A=@n14.A, B=LinearInterpolate(A=Multiply(A=@n100, B=@n101), Alpha=NamedRerouteUsage()))
  Reroute(None=Multiply(A=Multiply(A=@n37, B=@n102), B=@n38))
  n103 = Time()
  SmoothStep(Value=Frac(None=Multiply(A=@n103) {constB=0.5})) {constMax=0.85, constMin=0.15}
 ]
}
