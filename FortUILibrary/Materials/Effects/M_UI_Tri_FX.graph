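# Translucent UI material that draws animated triangle particles scrolling in a direction set by a rotation-fraction parameter.
# Three panners scroll one triangle mask texture at separate speeds, sampling its B, G and R channels as layers.
# A parabolic envelope along the rotated U axis, scaled by TriScale, drives each layer's SmoothStep threshold, thinning triangles toward the ends.
# Differently clamped layer sums blend Grad into Emphasis color, while their full sum is scaled by a U-axis ramp.
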Material /FortUILibrary/Materials/Effects/M_UI_Tri_FX
graph {
 properties {
  bUsedWithStaticMesh=false
  blendMode=BLEND_Translucent
  materialDomain=MD_UI
  pixelDepthOffsetMode=PDOM_Legacy
  usedShadingModels=Default Lit
 }
 nodes [
  n1 = MaterialFunctionCall!(Rotation_Fraction=ScalarParameter() {parameterName=DirectionRotationFraction}, TexCoord=TextureCoordinate()) {materialFunction=/FortUILibrary/Foundation/Art_ToolKit/AngledBoxes/MF_UI_Rotator_Fractions}
  n2 = ScalarParameter() {defaultValue=0.25, parameterName=Speed1}
  n3 = Constant()
  n4 = TextureSampleParameter2D(UVs=Panner(Coordinate=@n1.Result, Speed=AppendVector(A=@n2, B=@n3)) {speedX=0.25}) {parameterName=TextureMask, samplerType=SAMPLERTYPE_LinearColor, texture=/Game/UI/Foundation/Items/Textures/T-Card-Hero-Anim-Triangle}
  n5 = ComponentMask(None=@n1.Result) {r=true}
  n6 = Multiply(A=@n5, B=OneMinus(None=@n5))
  n7 = Multiply(A=Constant() {r=4}, B=ScalarParameter() {defaultValue=1, parameterName=TriScale})
  n8 = Saturate(None=OneMinus(None=Multiply(A=@n6, B=@n7) {constB=4}))
  n9 = MaterialFunctionCall(Alpha=@n4.B, Min=@n8) {materialFunction=&SmoothStep}
  n10 = Multiply(A=Subtract(A=Multiply(A=@n9.Result)) {constB=0.65}) {constB=8}
  n11 = Clamp(None=@n10)
  n12 = ScalarParameter() {defaultValue=0.15, parameterName=Speed2}
  n13 = Constant()
  n14 = TextureSampleParameter2D(UVs=Panner(Coordinate=@n1.Result, Speed=AppendVector(A=@n12, B=@n13)) {speedX=0.15}) {parameterName=TextureMask, samplerType=SAMPLERTYPE_LinearColor, texture=&T-Card-Hero-Anim-Triangle}
  n15 = MaterialFunctionCall(Alpha=@n14.G, Min=@n8) {materialFunction=&SmoothStep}
  n16 = Multiply(A=Subtract(A=Multiply(A=@n15.Result)) {constB=0.65}) {constB=8}
  n17 = Clamp(None=@n16) {maxDefault=0.5}
  n18 = ScalarParameter() {defaultValue=0.5, parameterName=Speed3}
  n19 = Constant()
  n20 = TextureSampleParameter2D(UVs=Panner(Coordinate=@n1.Result, Speed=AppendVector(A=@n18, B=@n19)) {speedX=0.5}) {parameterName=TextureMask, samplerType=SAMPLERTYPE_LinearColor, texture=&T-Card-Hero-Anim-Triangle}
  n21 = MaterialFunctionCall(Alpha=@n20.R, Min=@n8) {materialFunction=&SmoothStep}
  n22 = Multiply(A=Subtract(A=Multiply(A=@n21.Result)) {constB=0.65}) {constB=8}
  LinearInterpolate(
   A=VectorParameter() {parameterName=Grad, defaultValue=<rgba 0.8542 0.5907 0 1>}.RGB
   B=VectorParameter() {parameterName=Emphasis, defaultValue=<rgba 1 0.8541 0.2474 1>}.RGB
   Alpha=Add(A=Add(A=@n11, B=@n17), B=Clamp(None=@n22) {maxDefault=0.05})
  )
  n23 = Clamp(None=@n10)
  n24 = Clamp(None=@n16)
  LinearInterpolate(B=Add(A=Add(A=@n23, B=@n24), B=Clamp(None=@n22)), Alpha=Multiply(A=@n5) {constB=3})
 ]
}
