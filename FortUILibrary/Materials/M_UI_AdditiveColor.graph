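# M_UI_AdditiveColor is a UI-domain material with additive blending, named for adding a flat color over what lies behind it.
# Its only node is a Color vector parameter with a dark grey default, which material instances can override.
# The graph has no textures or math, and shows no link from the parameter to a material output.
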Material /FortUILibrary/Materials/M_UI_AdditiveColor
graph {
 properties {bUsedWithStaticMesh=false, blendMode=BLEND_Additive, materialDomain=MD_UI, pixelDepthOffsetMode=PDOM_Legacy, usedShadingModels=Default Lit}
 nodes [VectorParameter() {parameterName=Color, defaultValue=<rgba 0.2423 0.2423 0.2423 1>}]
}
