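# Translucent UI material that draws a skewed box with separately colored fill and border, each with its own alpha.
# Outer and inner box-skew functions produce masks softened by SmoothStep with a Smoothness parameter.
# Subtracting inner from outer yields the border mask, which blends InnerColor toward OuterColor.
# The inner box scale derives from saturated aspect ratios of UV channel 3, adjusting the inset to the box's proportions.
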Material /FortUILibrary/Materials/M_UI_BoxSkew
graph {
 properties {
  bUsedWithStaticMesh=false
  blendMode=BLEND_Translucent
  materialDomain=MD_UI
  pixelDepthOffsetMode=PDOM_Legacy
  usedShadingModels=Default Lit
 }
 nodes [
  n1 = VectorParameter() {parameterName=InnerColor, defaultValue=<rgba 0.3372 0.3372 0.3372 1>}
  n2 = Constant()
  n3 = Multiply(A=ScalarParameter() {defaultValue=0.01, parameterName=Smoothness}, B=NamedRerouteUsage())
  n4 = ScalarParameter() {parameterName=MinWidth}
  n5 = SmoothStep(
   Min=@n2
   Max=@n3
   Value=MaterialFunctionCall!(
    MinWidth=@n4
    Angle Top=ScalarParameter() {group=Outer Skew Box, parameterName=O_Angle_Top}
    Angle Right=ScalarParameter() {group=Outer Skew Box, parameterName=O_Angle_Right}
    Angle Bottom=ScalarParameter() {group=Outer Skew Box, parameterName=O_Angle_Bottom}
    Angle Left=ScalarParameter() {group=Outer Skew Box, parameterName=O_Angle_Left}
    Scale=ScalarParameter() {defaultValue=1, group=Outer Skew Box, parameterName=O_SkewScale}
    Flip=StaticBool()
   ) {materialFunction=/Game/UI/Sidebar/Materials/MF_BoxSkew}.Result
  )
  n6 = VectorParameter() {parameterName=OuterColor, defaultValue=<rgba 1 1 1 1>}
  n7 = ScalarParameter() {group=Inner Skew Box, parameterName=I_Angle_Top}
  n8 = ScalarParameter() {group=Inner Skew Box, parameterName=I_Angle_Right}
  n9 = ScalarParameter() {group=Inner Skew Box, parameterName=I_Angle_Bottom}
  n10 = ScalarParameter() {group=Inner Skew Box, parameterName=I_Angle_Left}
  n11 = ScalarParameter() {group=Inner Skew Box, parameterName=I_Offset_Right}
  n12 = ScalarParameter() {group=Inner Skew Box, parameterName=I_Offset_Left}
  n13 = TextureCoordinate() {coordinateIndex=3}
  n14 = ComponentMask(None=@n13) {g=true}
  n15 = ComponentMask(None=@n13) {r=true}
  n16 = Divide(A=@n14, B=@n15)
  n17 = Divide(A=@n15, B=@n14)
  n18 = AppendVector(A=@n16, B=@n17)
  n19 = OneMinus(None=Multiply(A=ScalarParameter() {defaultValue=0.4, group=Inner Skew Box, parameterName=I_SkewScale}, B=Saturate(None=@n18)))
  n20 = StaticBool()
  n21 = Subtract(
   A=@n5
   B=SmoothStep(
    Min=@n2
    Max=@n3
    Value=MaterialFunctionCall!(
     MinWidth=@n4
     Angle Top=@n7
     Angle Right=@n8
     Angle Bottom=@n9
     Angle Left=@n10
     "Offset  Right"=@n11
     "Offset  Left"=@n12
     Scale=@n19
     Flip=@n20
    ) {materialFunction=/Game/UI/Sidebar/Materials/MF_BoxSkew1}.Result
   )
  )
  LinearInterpolate(A=Multiply(A=@n1.RGB, B=@n5), B=@n6.RGB, Alpha=@n21)
  NamedRerouteDeclaration(None=@n16) {name=ScaleX, nodeColor=<rgba 0 0 0 1>}
  NamedRerouteDeclaration(None=@n17) {name=ScaleY, nodeColor=<rgba 0 0 0 1>}
  Saturate(None=LinearInterpolate(A=Multiply(A=@n1.A, B=@n5), B=@n6.A, Alpha=@n21))
 ]
}
