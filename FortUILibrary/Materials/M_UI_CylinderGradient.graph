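# M_UI_CylinderGradient is a translucent UI material that builds a two-color gradient with MF_UI_ReflectedGradient.
# Offset and rotation parameters drive the function, with the rotation center built from Center X twice, leaving Center Y unused.
# The result, optionally inverted by a static switch, blends ColorA and ColorB in separate RGB and alpha lerps.
# GradientWidth, GradientDensity and GradientPower parameters exist but show no connections.
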Material /FortUILibrary/Materials/M_UI_CylinderGradient
graph {
 properties {
  bUsedWithStaticMesh=false
  blendMode=BLEND_Translucent
  materialDomain=MD_UI
  pixelDepthOffsetMode=PDOM_Legacy
  usedShadingModels=Default Lit
 }
 nodes [
  n1 = VectorParameter() {group=Color Settings, parameterName=ColorA, sortPriority=0}
  n2 = VectorParameter() {group=Color Settings, parameterName=ColorB, sortPriority=1, defaultValue=<rgba 1 1 1 1>}
  n3 = VectorParameter() {
   group=Rotation Settings
   parameterName=GradientRotationCenter
   sortPriority=1
   channelNames {a=Do Not Use, b=Do Not Use, g=Center Y, r=Center X}
   defaultValue=<rgba 0.5 0.5 0 0>
  }
  n4 = MaterialFunctionCall!(
   UVs
   Offset=ScalarParameter() {group=Gradient Settings, parameterName=GradientOffset, sliderMax=1, sliderMin=-1, sortPriority=2}
   RotationCenter=AppendVector(A=@n3.Center X, B=@n3.Center X)
   RotationAngle=ScalarParameter() {group=Rotation Settings, parameterName=GradientRotation, sliderMax=1, sortPriority=0}
  ) {materialFunction=/FortUILibrary/MaterialFunctions/Gradients/MF_UI_ReflectedGradient}
  n5 = StaticSwitchParameter(True=OneMinus(None=@n4.Result), False=@n4.Result) {group=Gradient Settings, parameterName=InvertGradient, sortPriority=3}
  LinearInterpolate(A=@n1.RGB, B=@n2.RGB, Alpha=@n5)
  LinearInterpolate(A=@n1.A, B=@n2.A, Alpha=@n5)
  ScalarParameter() {defaultValue=0.25, group=Gradient Settings, parameterName=GradientWidth, sliderMax=1, sortPriority=0}
  ScalarParameter() {defaultValue=1, group=Gradient Settings, parameterName=GradientDensity, sliderMax=10, sortPriority=1}
  ScalarParameter() {defaultValue=2, group=Gradient Settings, parameterName=GradientPower}
 ]
}
