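# Translucent UI material that renders a three-color linear gradient from ColorA, ColorB and ColorC parameters.
# MF_UI_LinearGradient_3Colors produces the gradient alpha from start, mid, end, offset, rotation angle and rotation center parameters.
# That alpha drives two Lerp_3Color calls, blending RGB and alpha separately so each color stop can carry its own transparency.
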Material /FortUILibrary/Materials/M_UI_LinearGradient_3Color
graph {
 properties {
  bUsedWithStaticMesh=false
  blendMode=BLEND_Translucent
  materialDomain=MD_UI
  pixelDepthOffsetMode=PDOM_Legacy
  usedShadingModels=Default Lit
 }
 nodes [
  n1 = VectorParameter() {group=Color Settings, parameterName=ColorA, defaultValue=<rgba 1 0 0 1>}
  n2 = VectorParameter() {group=Color Settings, parameterName=ColorB, defaultValue=<rgba 0 1 0 1>}
  n3 = VectorParameter() {group=Color Settings, parameterName=ColorC, defaultValue=<rgba 0 0 1 1>}
  n4 = VectorParameter() {
   group=Gradient Settings
   parameterName=GradientRotationCenter
   sortPriority=5
   channelNames {a="", b="", g=Center Y, r=Center X}
   defaultValue=<rgba 0.5 0.5 0 0>
  }
  n5 = MaterialFunctionCall!(
   Texture
   Gradient Start=ScalarParameter() {group=Gradient Settings, parameterName=GradientStart, sortPriority=1}
   Gradient Mid=ScalarParameter() {defaultValue=0.5, group=Gradient Settings, parameterName=GradientMid, sortPriority=2}
   Gradient End=ScalarParameter() {defaultValue=1, group=Gradient Settings, parameterName=GradientEnd, sortPriority=2}
   Rotation Angle=ScalarParameter() {group=Gradient Settings, parameterName=GradientRotation, sortPriority=4}
   Rotation Center=@n4
   Gradient Offset=ScalarParameter() {group=Gradient Settings, parameterName=GradientOffset, sortPriority=3}
  ) {materialFunction=/FortUILibrary/MaterialFunctions/Gradients/MF_UI_LinearGradient_3Colors}
  MaterialFunctionCall(A=@n1.RGB, B=@n2.RGB, C=@n3.RGB, Alpha=@n5.Result) {materialFunction=&Lerp_3Color}
  MaterialFunctionCall(A=@n1.A, B=@n2.A, C=@n3.A, Alpha=@n5.Result) {materialFunction=&Lerp_3Color}
 ]
}
