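# M_UI_Texture is a translucent UI-domain material built around a swappable texture with an adjustable alpha.
# It multiplies the alpha channel of the Texture parameter by the scalar Alpha parameter, which defaults to fully on.
# Material instances can therefore change the image and scale its alpha, while the graph as listed does not show which material outputs are connected.
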Material /FortUILibrary/Materials/M_UI_Texture
graph {
 properties {
  bUsedWithStaticMesh=false
  blendMode=BLEND_Translucent
  materialDomain=MD_UI
  pixelDepthOffsetMode=PDOM_Legacy
  usedShadingModels=Default Lit
 }
 nodes [Multiply(A=TextureSampleParameter2D() {parameterName=Texture}.A, B=ScalarParameter() {defaultValue=1, parameterName=Alpha, sliderMax=1})]
}
