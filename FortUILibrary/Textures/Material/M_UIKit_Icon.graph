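# UI material for UIKit icons, rendering a distance-field texture with button-state color and opacity.
# A Use MSDF switch picks between a channel-packed SDF (red fill, green outline) and an MSDF converted to SDF, both antialiased with optional outline.
# Each path has offset, scale and rotation UVs.
# Modular button functions blend base and selectable states, while alpha-composite output uses Brightness and Additive percentages to control glow versus coverage.
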Material /FortUILibrary/Textures/Material/M_UIKit_Icon
graph {
 properties {
  bUsedWithStaticMesh=false
  blendMode=BLEND_AlphaComposite
  materialDomain=MD_UI
  pixelDepthOffsetMode=PDOM_Legacy
  usedShadingModels=Default Lit
 }
 nodes [
  n1 = StaticSwitchParameter(True=NamedRerouteUsage(), False=NamedRerouteUsage()) {group=2 - Distance Field, parameterName=Use MSDF, sortPriority=0}
  n2 = StaticSwitchParameter(True=NamedRerouteUsage(), False=NamedRerouteUsage()) {group=2 - Distance Field, parameterName=Use MSDF, sortPriority=0}
  n3 = Multiply(A=ScalarParameter() {defaultValue=1, group=4- Alpha Composite, parameterName=Brightness Percentage, sliderMax=1, sortPriority=2}, B=@n2)
  n4 = NamedRerouteUsage()
  Multiply(A=Multiply(A=VertexColor().A, B=Multiply(A=@n1, B=@n3)), B=@n4)
  n5 = ScalarParameter() {group=4- Alpha Composite, parameterName=Additive Percentage, sliderMax=1, sortPriority=3}
  Multiply(A=@n4, B=Multiply(A=@n2, B=OneMinus(None=@n5)))
  n6 = MaterialFunctionCall!(Idle, Idle Opacity, Hovered, Hovered Opacity, Focused, Focused Opacity, Pressed, Pressed Opacity, Disabled, Disabled Opacity) {materialFunction=/FortUILibrary/Systems/ModularButton/MaterialFunctions/MF_UIKit_ButtonStatesBase_Lerp}
  n7 = MaterialFunctionCall!(
   Base States=@n6.Color
   Base States Opacity=@n6.Color
   Selected_Idle
   Selected_Idle Opacity
   Selected_Hovered
   Selected_Hovered Opacity
   Selected_Focused
   Selected_Focused Opacity
   Selected_Pressed
   Selected_Pressed Opacity
   Selected_Disabled
   Selected_Disabled Opacity
  ) {materialFunction=/FortUILibrary/Systems/ModularButton/MaterialFunctions/MF_UIKit_ButtonStatesSelectable_Lerp}
  NamedRerouteDeclaration(None=@n7.Color) {name=State Color, nodeColor=<rgba 0 0 0 1>}
  NamedRerouteDeclaration(None=@n7.Opacity) {name=State Opacity, nodeColor=<rgba 0 0 0 1>}
  n8 = TextureSampleParameter2D(UVs=NamedRerouteUsage()) {
   group=2 - Distance Field
   parameterName=Icon
   samplerSource=SSM_Clamp_WorldGroupSettings
   samplerType=SAMPLERTYPE_LinearColor
   sortPriority=1
   texture=/FortUILibrary/Textures/Icons/T_UI_Icon_UIKit_Search_SDF
  }
  NamedRerouteDeclaration(None=AppendVector(A=@n8.RGB, B=Constant())) {name=Texture Channels, nodeColor=<rgba 0 0 0 1>}
  n9 = LinearInterpolate(B=VectorParameter() {group=3 - Outline Settings, parameterName=Outline Color, defaultValue=<rgba 0 0 0 1>}.RGB, Alpha=NamedRerouteUsage())
  n10 = StaticSwitchParameter(True=@n9, False=Constant()) {group=3 - Outline Settings, parameterName=Display Outline}
  NamedRerouteDeclaration(None=LinearInterpolate(A=LinearInterpolate(B=@n9, Alpha=@n10), B=NamedRerouteUsage(), Alpha=NamedRerouteUsage())) {name=MSDF Colors, nodeColor=<rgba 1 0.7734 0 1>}
  n11 = MaterialFunctionCall() {materialFunction=&GetUserInterfaceUV}
  n12 = AppendVector(
   A=Subtract(A=ComponentMask(None=@n11.Normalized UV) {r=true}, B=ScalarParameter() {group=1 - Transform, parameterName=OffsetX, sortPriority=48})
   B=Add(A=ComponentMask(None=@n11.Normalized UV) {g=true}, B=ScalarParameter() {group=1 - Transform, parameterName=OffsetY, sortPriority=48})
  )
  n13 = Multiply(
   A=ScalarParameter() {defaultValue=1, group=1 - Transform, parameterName=MSDF Scale}
   B=AppendVector(
    A=ScalarParameter() {defaultValue=1, group=1 - Transform, parameterName=ScaleX}
    B=ScalarParameter() {defaultValue=1, group=1 - Transform, parameterName=ScaleY}
   )
  )
  NamedRerouteDeclaration(None=MaterialFunctionCall(
    UVs=MaterialFunctionCall(UVs=@n12, Texture Scale=@n13) {materialFunction=&ScaleUVsByCenter}.UVs
    "Rotation Angle (0-1)"=ScalarParameter() {group=1 - Transform, parameterName=Rotation}
   ) {materialFunction=&CustomRotator}.Rotated Values) {name=MSDF UV, nodeColor=<rgba 1 0.3984 0 1>}
  n14 = ScalarParameter() {defaultValue=0.25, group=3 - Outline Settings, parameterName=Outline Opacity, sliderMax=1, sortPriority=34}
  n15 = NamedRerouteUsage()
  NamedRerouteDeclaration(None=LinearInterpolate(
    A=Multiply(A=@n14, B=@n15)
    B=ScalarParameter() {defaultValue=1, group=1 - Transform, parameterName=Opacity, sliderMax=1}
    Alpha=NamedRerouteUsage()
   ) {constAlpha=1}) {name=Opacity, nodeColor=<rgba 0 0 0 1>}
  n16 = VectorParameter() {group=3 - Outline Settings, parameterName=Outline Color, defaultValue=<rgba 0 0 0 1>}
  n17 = Multiply(A=Constant() {r=1}, B=NamedRerouteUsage())
  n18 = Constant() {r=1}
  n19 = NamedRerouteUsage()
  NamedRerouteDeclaration(None=LinearInterpolate(A=LinearInterpolate(B=@n16.RGB, Alpha=@n17), B=NamedRerouteUsage(), Alpha=Multiply(A=@n18, B=@n19))) {name=Separated Colors, nodeColor=<rgba 0 0 0 1>}
  n20 = NamedRerouteUsage()
  n21 = ScalarParameter() {group=2 - Distance Field, parameterName=Smoothing, sliderMax=1}
  n22 = MaterialFunctionCall(
   DistanceField=ComponentMask(None=@n20) {r=true}
   Threshold=ScalarParameter() {defaultValue=0.68, group=2 - Distance Field, parameterName=Fill Aliasing}
   Smoothing=@n21
  ) {materialFunction=&MF_UI_AntiAliasing}
  n23 = ScalarParameter() {defaultValue=0.35, group=3 - Outline Settings, parameterName=Outline Scale, sliderMax=1}
  n24 = StaticSwitchParameter(
   True=MaterialFunctionCall(DistanceField=ComponentMask(None=@n20) {g=true}, Threshold=OneMinus(None=@n23), Smoothing=@n21) {materialFunction=&MF_UI_AntiAliasing}.Outer
   False=Constant()
  ) {group=3 - Outline Settings, parameterName=Display Outline}
  n25 = MaterialFunctionCall(Float2=NamedRerouteDeclaration(None=AppendVector(A=@n22.Outer, B=@n24)) {name=Distance Field Pass, nodeColor=<rgba 0 0 0 1>}) {materialFunction=&BreakOutFloat2Components}
  NamedRerouteDeclaration(None=@n25.R) {name=Red Channel, nodeColor=<rgba 0 0 0 1>}
  NamedRerouteDeclaration(None=@n25.G) {name=Green Channel, nodeColor=<rgba 0 0 0 1>}
  n26 = TextureSampleParameter2D(UVs=NamedRerouteUsage()) {
   group=2 - Distance Field
   parameterName=MSDF
   samplerType=SAMPLERTYPE_LinearColor
   texture=/FortUILibrary/Textures/Icons/T_UI_Icon_UIKit_Search_MSDF
  }
  n27 = MaterialFunctionCall!(MSDF RGB=@n26.RGB) {materialFunction=/FortUILibrary/MaterialFunctions/Shapes/Modifiers/MF_UI_MSDFtoSDF}
  n28 = Constant() {r=0.5}
  NamedRerouteDeclaration(None=MaterialFunctionCall(DistanceField=@n27.SDF, Threshold=@n28, Smoothing=ScalarParameter() {group=2 - Distance Field, parameterName=Smoothing}) {materialFunction=&MF_UI_AntiAliasing}.Outer) {name=Icon Mask, nodeColor=<rgba 0.9688 0 1 1>}
  n29 = MaterialFunctionCall() {materialFunction=&GetUserInterfaceUV}
  n30 = AppendVector(
   A=Subtract(A=ComponentMask(None=@n29.Normalized UV) {r=true}, B=ScalarParameter() {group=1 - Transform, parameterName=OffsetX, sortPriority=48})
   B=Add(A=ComponentMask(None=@n29.Normalized UV) {g=true}, B=ScalarParameter() {group=1 - Transform, parameterName=OffsetY, sortPriority=48})
  )
  n31 = ScalarParameter() {defaultValue=2, group=1 - Transform, parameterName=Scale, sortPriority=0}
  NamedRerouteDeclaration(None=MaterialFunctionCall(
    UVs=MaterialFunctionCall(UVs=@n30, Texture Scale=@n31) {materialFunction=&ScaleUVsByCenter}.UVs
    "Rotation Angle (0-1)"=ScalarParameter() {group=1 - Transform, parameterName=Rotation}
   ) {materialFunction=&CustomRotator}.Rotated Values) {name=UVs, nodeColor=<rgba 0 0 0 1>}
  n32 = ScalarParameter() {defaultValue=0.35, group=3 - Outline Settings, parameterName=Outline Scale, sliderMax=1}
  NamedRerouteDeclaration(None=MaterialFunctionCall(
    DistanceField=@n27.SDF
    Threshold=Subtract(A=@n28, B=@n32) {constB=0.2}
    Smoothing=ScalarParameter() {group=2 - Distance Field, parameterName=Smoothing}
   ) {materialFunction=&MF_UI_AntiAliasing}.Outer) {name=Icon Outline Mask, nodeColor=<rgba 0 1 0.6484 1>}
  n33 = Multiply(
   A=NamedRerouteUsage()
   B=ScalarParameter() {defaultValue=0.25, group=3 - Outline Settings, parameterName=Outline Opacity, sliderMax=1, sortPriority=34}
  ) {constB=0.5}
  n34 = StaticSwitchParameter(True=@n33, False=Constant()) {group=3 - Outline Settings, parameterName=Display Outline}
  NamedRerouteDeclaration(None=LinearInterpolate(
    A=LinearInterpolate(B=@n33, Alpha=@n34)
    B=ScalarParameter() {defaultValue=1, group=1 - Transform, parameterName=Opacity, sliderMax=1}
    Alpha=NamedRerouteUsage()
   )) {name=MSDF Opacity, nodeColor=<rgba 0 0.6484 1 1>}
 ]
}
