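# UI material that draws a circular download progress indicator, with a background ring, colored progress arc, glow and optional fill.
# Circle SDFs form the anti-aliased rings, and a radial arc function driven by the Progress parameter supplies the fill mask and gradient.
# Two circle SDFs cap the arc ends, the end one rotated by progress, with feathered SDFs for glow.
# A time-rotated radial gradient provides an animated feedback highlight, toggled by Display Feedback.
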Material /FortUILibrary/Textures/Material/M_UI_DownloadProgress
graph {
 properties {
  bAutomaticallySetUsageInEditor=false
  bUsedWithStaticMesh=false
  blendMode=BLEND_Translucent
  materialDomain=MD_UI
  usedShadingModels=Default Lit
 }
 nodes [
  n1 = NamedRerouteDeclaration(None=VectorParameter() {parameterName=Progress Background Color, defaultValue=<rgba 0.3118 0.4323 0.105 1>}.RGB) {name=Background Color, nodeColor=<rgba 0 0 0 1>}
  n2 = LinearInterpolate(A=@n1, B=Constant3Vector() {constant=<rgba 0.0048 0.004 0.0116 1>}, Alpha=NamedRerouteUsage()) {constB=0}
  n3 = NamedRerouteUsage()
  n4 = Constant3Vector() {constant=<rgba 0 0.8714 1 1>}
  n5 = NamedRerouteUsage()
  n6 = Subtract(A=NamedRerouteUsage(), B=NamedRerouteUsage())
  n7 = Multiply(A=@n5, B=LinearInterpolate(Alpha=Saturate(None=@n6)) {constA=1, constB=0.25})
  n8 = ScalarParameter() {parameterName=Fill Opacity, sliderMax=1}
  n9 = LinearInterpolate(
   A=NamedRerouteDeclaration(None=Constant3Vector() {constant=<rgba 0 0.5395 1 1>}) {name=End Color, nodeColor=<rgba 0 0 0 1>}
   B=NamedRerouteDeclaration(None=Constant3Vector() {constant=<rgba 0 0.7454 1 1>}) {name=Start Color, nodeColor=<rgba 0 0 0 1>}
   Alpha=NamedRerouteUsage()
  )
  n10 = SmoothStep(Value=NamedRerouteUsage()) {constMax=0.375}
  n11 = LinearInterpolate(
   A=LinearInterpolate(B=NamedRerouteDeclaration(None=@n9) {name=Progress Base Color, nodeColor=<rgba 0 0 0 1>}, Alpha=Saturate(None=@n10)) {constA=1}
   B=LinearInterpolate(A=NamedRerouteUsage(), B=NamedRerouteUsage(), Alpha=NamedRerouteUsage())
   Alpha=NamedRerouteUsage()
  )
  n12 = NamedRerouteUsage()
  n13 = LinearInterpolate(
   A=LinearInterpolate(
    A=LinearInterpolate(A=@n2, B=@n1, Alpha=@n3) {constB=0}
    B=NamedRerouteDeclaration(None=@n4) {name=Glow Color, nodeColor=<rgba 0 0 0 1>}
    Alpha=LinearInterpolate(A=@n5, B=@n7, Alpha=@n8)
   )
   B=NamedRerouteDeclaration(None=LinearInterpolate(A=@n11, Alpha=@n12)) {name=Progress Color, nodeColor=<rgba 0 0 0 1>}
   Alpha=NamedRerouteUsage()
  )
  n14 = NamedRerouteUsage()
  n15 = ScalarParameter() {defaultValue=1, group=Style, parameterName=Display Feedback, sliderMax=1}
  n16 = NamedRerouteUsage()
  n17 = NamedRerouteUsage()
  LinearInterpolate(
   A=@n13
   B=LinearInterpolate(A=@n13, B=LinearInterpolate(B=@n14, Alpha=@n15), Alpha=LinearInterpolate(Alpha=@n16) {constB=0.25}) {constAlpha=0.25}
   Alpha=Ceil(None=OneMinus(None=@n17))
  )
  n18 = Multiply(A=NamedRerouteUsage(), B=ScalarParameter() {parameterName=Fill Opacity, sliderMax=1}) {constB=0.5}
  n19 = NamedRerouteUsage()
  n20 = Multiply(A=ScalarParameter() {defaultValue=0.7, parameterName=Background Progress Opacity, sliderMax=1}, B=@n19)
  n21 = Multiply(A=NamedRerouteUsage(), B=NamedRerouteUsage())
  LinearInterpolate(
   A=LinearInterpolate(A=LinearInterpolate(A=@n18, B=@n20, Alpha=@n19), Alpha=NamedRerouteDeclaration(None=@n21) {name=Progress Bar Mask, nodeColor=<rgba 0 0 0 1>})
   Alpha=NamedRerouteUsage()
  )
  NamedRerouteDeclaration(None=MaterialFunctionCall() {materialFunction=&GetUserInterfaceUV}.Normalized UV) {name=UV, nodeColor=<rgba 0 0 0 1>}
  n22 = NamedRerouteDeclaration(None=ScalarParameter() {defaultValue=0.1, parameterName=Glow Feathering, sliderMax=1}) {name=Glow Feathering, nodeColor=<rgba 0 0 0 1>}
  n23 = Multiply(A=@n22) {constB=-1}
  n24 = Saturate(None=SmoothStep(Min=@n23, Max=@n22, Value=NamedRerouteUsage()))
  n25 = Saturate(None=SmoothStep(Min=@n23, Max=@n22, Value=NamedRerouteUsage()))
  n26 = Max(
   A=NamedRerouteDeclaration(None=OneMinus(None=@n24)) {name=Start Point Glow Mask, nodeColor=<rgba 0 0 0 1>}
   B=NamedRerouteDeclaration(None=OneMinus(None=@n25)) {name=End Point Glow Mask, nodeColor=<rgba 0 0 0 1>}
  )
  n27 = NamedRerouteUsage()
  n28 = OneMinus(None=Saturate(None=SmoothStep(Min=@n23, Max=@n22, Value=@n27)))
  n29 = NamedRerouteUsage()
  n30 = ScalarParameter() {defaultValue=1, parameterName=Glow Opacity, sliderMax=1}
  n31 = Ceil(None=NamedRerouteUsage())
  NamedRerouteDeclaration(None=Multiply(A=LinearInterpolate(A=@n26, B=@n28, Alpha=@n29) {constB=0}, B=Multiply(A=@n30, B=@n31))) {name=Glow Mask, nodeColor=<rgba 0 0 0 1>}
  n32 = MaterialFunctionCall(
   UVs=NamedRerouteUsage()
   Size=NamedRerouteDeclaration(None=ScalarParameter() {defaultValue=0.9, parameterName=Size}) {name=Size, nodeColor=<rgba 0 0 0 1>}
  ) {materialFunction=&MF_UI_Circle}
  n33 = NamedRerouteDeclaration(None=ScalarParameter() {defaultValue=0.05, parameterName=Outline BG Thickness}) {name=Outline BG Thickness, nodeColor=<rgba 0 0 0 1>}
  n34 = Abs(None=Add(A=NamedRerouteDeclaration(None=@n32.SDF) {name=Shape SDF, nodeColor=<rgba 0 0 0 1>}, B=@n33))
  n35 = ScalarParameter() {defaultValue=0.05, parameterName=Outline Progress Thickness, sliderMax=1}
  NamedRerouteDeclaration(None=Subtract(A=@n34, B=NamedRerouteDeclaration(None=@n35) {name=Outline Progress Thickness, nodeColor=<rgba 0 0 0 1>})) {name=Progress SDF, nodeColor=<rgba 0 0 0 1>}
  NamedRerouteDeclaration(None=MaterialFunctionCall(DistanceField=NamedRerouteUsage()) {materialFunction=&MF_UI_AntiAliasing}.Inner) {name=Progression LVL Mask, nodeColor=<rgba 0 0 0 1>}
  n36 = ScalarParameter() {defaultValue=0.125, parameterName=Progress, sliderMax=1}
  n37 = MaterialFunctionCall!(
   UV=NamedRerouteUsage()
   "Arc Start Angle (0-1)"
   "Arc End Angle (0-1)"
   "Fill Ratio (0-1)"=OneMinus(None=NamedRerouteDeclaration(None=@n36) {name=Progress Value, nodeColor=<rgba 0 0 0 1>})
   Fill Clockwise=Constant() {r=1}
  ) {materialFunction=/FortUILibrary/MaterialFunctions/Basics/MF_UI_RadialArc}
  n38 = Constant()
  n39 = NamedRerouteUsage()
  NamedRerouteDeclaration(None=Saturate(None=MaterialFunctionCall(Input=@n37.Radial Fill, Input Low=@n38, Input High=@n39) {materialFunction=&RemapValue_0-1}.Result)) {name=Progress Gradient Range, nodeColor=<rgba 0 0 0 1>}
  NamedRerouteDeclaration(None=Subtract(A=@n34, B=@n33) {constB=0.1}) {name=Progress Background SDF, nodeColor=<rgba 0 0 0 1>}
  NamedRerouteDeclaration(None=MaterialFunctionCall(DistanceField=NamedRerouteUsage()) {materialFunction=&MF_UI_AntiAliasing}.Inner) {name=Ring Mask, nodeColor=<rgba 0 0 0 1>}
  n40 = NamedRerouteDeclaration(None=MaterialFunctionCall(DistanceField=NamedRerouteUsage()) {materialFunction=&MF_UI_AntiAliasing}.Inner) {name=Start Point Mask, nodeColor=<rgba 0 0 0 1>}
  n41 = NamedRerouteDeclaration(None=MaterialFunctionCall(DistanceField=NamedRerouteUsage()) {materialFunction=&MF_UI_AntiAliasing}.Inner) {name=End Point Mask, nodeColor=<rgba 0 0 0 1>}
  n42 = Ceil(None=@n37.Radial Fill)
  NamedRerouteDeclaration(None=LinearInterpolate(A=Max(A=@n40, B=@n41), Alpha=NamedRerouteDeclaration(None=@n42) {name=Progress Mask, nodeColor=<rgba 0 0 0 1>})) {name=Progress Mask with points, nodeColor=<rgba 0 0 0 1>}
  NamedRerouteDeclaration(None=@n37.Arc Mask) {name=Arc Mask, nodeColor=<rgba 0 0 0 1>}
  n43 = Constant2Vector() {g=-0.5}
  n44 = NamedRerouteUsage()
  n45 = NamedRerouteUsage()
  n46 = AppendVector(A=@n43.R, B=Add(A=Multiply(A=@n43.G, B=@n44), B=Multiply(A=@n45) {constB=0.5}))
  n47 = Subtract(A=NamedRerouteUsage(), B=@n46)
  n48 = NamedRerouteUsage()
  n49 = Ceil(None=NamedRerouteUsage())
  NamedRerouteDeclaration(None=Multiply(A=MaterialFunctionCall(UVs=@n47, Size=@n48) {materialFunction=&MF_UI_Circle}.SDF, B=@n49)) {name=Start Point, nodeColor=<rgba 0 0 0 1>}
  n50 = NamedRerouteUsage()
  n51 = Subtract(A=Rotator(Coordinate=NamedRerouteUsage(), Time=OneMinus(None=@n50)) {speed=6.2832}, B=@n46)
  NamedRerouteDeclaration(None=Multiply(A=@n49, B=MaterialFunctionCall(UVs=@n51, Size=@n48) {materialFunction=&MF_UI_Circle}.SDF)) {name=End Point, nodeColor=<rgba 0 0 0 1>}
  n52 = NamedRerouteUsage()
  n53 = Multiply(A=Time()) {constB=-1}
  n54 = ComponentMask(None=MaterialFunctionCall!(Vector or UVs=Rotator(Coordinate=@n52, Time=@n53) {speed=6.2832}) {materialFunction=/FortUILibrary/MaterialFunctions/Basics/MF_UI_CheaperVectorToRadial}.Radial Coordinates) {r=true}
  n55 = Multiply(A=Min(A=@n54, B=OneMinus(None=@n54))) {constB=2}
  n56 = Constant() {r=0.5}
  n57 = Constant() {r=1}
  NamedRerouteDeclaration(None=SmoothStep(Value=MaterialFunctionCall(Input=@n55, Input Low=@n56, Input High=@n57) {materialFunction=&RemapValue_0-1}.Result)) {name=ProgressFeedback, nodeColor=<rgba 0 0 0 1>}
  NamedRerouteDeclaration(None=@n32.Fill) {name=Fill Mask, nodeColor=<rgba 0 0 0 1>}
 ]
}
